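; Gameplay ability that runs the cooldown for the DBNO building action in TrendConcentrate.
; On activation, a near-zero delay, noted as temporary, lets the build fire before a configurable gameplay effect is applied to the player pawn.
; It then waits for the InteractBuild cooldown tag to be removed and ends itself.
; On end, it removes effects matching the initial activation tag and applies the passive building grant effect.
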(event Ability|EventActivateAbility
  ;[1 0.02437 0 0.5] Allows the building to fire before cooldown TEMP
  (Utilities|FlowControl|Delay 1e-05)
  (GameplayEffects|ApplyGameplayEffectToSelf :self (Ability|GetAbilitySystemComponent (CallFunction|GetAvatarasFortPlayerPawn)) :GameplayEffectClass (Variables|Default|GetGameplay%20Effect%20Class) :Level 0.0)
  (Ability|Tasks|WaitGameplayTagRemove "(TagName=\"Action.Player.InteractBuild.TrendConcentrate.Cooldown\")"
    (:Removed
      (Ability|EndAbility))))

(event Ability|EventOnEndAbility
  (params (bWasCancelled :type Boolean))
  (bind _returnvalue (CallFunction|GetAvatarasFortPlayerPawn))
  (bind _returnvalue_2 (Ability|GetAbilitySystemComponent _returnvalue))
  (GameplayEffects|RemoveActiveEffectswithTags _returnvalue_2 (Variables|Default|GetInitialActivationTag))
  (GameplayEffects|ApplyGameplayEffectToSelf :self _returnvalue_2 :GameplayEffectClass "/TrendConcentrate/Gameplay/Effects/GE_DBNO_Building_PassiveGrant.GE_DBNO_Building_PassiveGrant_C" :Level 0.0))
