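; Passive gameplay ability that limits building by downed (DBNO) players to a pool of three points.
; A filtered event listener deducts a point per build, while a looping timer refunds one at a time.
; Emptying the pool pauses that timer, applies a cooldown effect and plays a cue; removal of the cooldown tag restores all points and resumes refills.
; A non-authority event updates the pip widget, and debug prints report point changes.
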(event Ability|EventActivateAbility
  (CallFunction|ListenforBuilds)
  (CallFunction|StartRefillTimer))

(event Custom|ListenForBuilds
  (bind _returnvalue (Ability|GetAvatarActorfromActorInfo))
  (Utilities|IsValid _returnvalue
    (:Is%20Valid
      (bind _returnvalue_2 (EventProcessorSubsystem|StartListeningToFilteredEvent 0 _returnvalue (Variables|Default|GetVerb%20Filter) false (EventDispatchers|CreateEvent :!event Custom|OnBuildConstructed)))
      (Variables|Default|SetListenerHandle _returnvalue_2))))

(event Custom|StartRefillTimer
  (bind _returnvalue (Utilities|Time|SetTimerbyEvent (EventDispatchers|CreateEvent :!event Custom|RefundBuildConstructed) (Variables|Default|GetBuildRefillRate) true))
  (Variables|Default|SetRefillTimerHandle _returnvalue))

(event Custom|RefundBuildConstructed
  (if (Health|IsDBNO (CallFunction|GetAvatarasFortPlayerPawn))
    (if (< (Variables|Default|GetBuildingPoints) 3.0)
      (bind _output_get (Variables|Default|SetBuildingPoints (+ (Variables|Default|GetBuildingPoints) 1.0)))
      (Development|PrintString "Build Points Refilled" true true "(R=0.026818,G=1.000000,B=0.037782,A=1.000000)" 5.0 "BuildPre")
      (Development|PrintString (Utilities|String|ToString(Float) _output_get) true true "(R=1.000000,G=0.198310,B=0.000000,A=1.000000)" 5.0 "Build#"))))

(event Custom|OnBuildConstructed
  (if (Health|IsDBNO (CallFunction|GetAvatarasFortPlayerPawn))
    (if (> (Variables|Default|GetBuildingPoints) 0)
      (bind _output_get (Variables|Default|SetBuildingPoints (- (Variables|Default|GetBuildingPoints) 1.0)))
      (Development|PrintString "Build Points Left" true true "(R=0.026818,G=1.000000,B=0.037782,A=1.000000)" 5.0 "BuildPre")
      (Development|PrintString (Utilities|String|ToString(Float) _output_get) true true "(R=1.000000,G=0.198310,B=0.000000,A=1.000000)" 5.0 "Build#")
      (if (<= (Variables|Default|GetBuildingPoints) 0)
        (Utilities|Time|PauseTimerbyHandle (Variables|Default|GetRefillTimerHandle))
        (GameplayEffects|ApplyGameplayEffectToSelf :self (Ability|GetAbilitySystemComponent (CallFunction|GetAvatarasFortPlayerPawn)) :GameplayEffectClass (Variables|Default|GetGameplay%20Effect%20Class) :Level 0.0)
        (Development|PrintString "Build Points Exhausted Building on Cooldown" true true "(R=1.000000,G=0.000000,B=0.000000,A=1.000000)" 5.0)
        (GameplayCue|ExecuteGameplayCue (CallFunction|GetAvatarasFortPlayerPawn) (Variables|Default|GetCooldownSoundCue))
        (Ability|Tasks|WaitGameplayTagRemove "(TagName=\"Action.Player.InteractBuild.TrendConcentrate.Cooldown\")"
          (:Removed
            (bind _output_get_2 (Variables|Default|SetBuildingPoints 3.0))
            (Development|PrintString "Build Points Refilled" true true "(R=0.026818,G=1.000000,B=0.037782,A=1.000000)" 5.0 "BuildPre")
            (Development|PrintString (Utilities|String|ToString(Float) _output_get_2) true true "(R=1.000000,G=0.198310,B=0.000000,A=1.000000)" 5.0 "Build#")
            (Utilities|Time|UnpauseTimerbyHandle (Variables|Default|GetRefillTimerHandle))
            (GameplayCue|ExecuteGameplayCue (CallFunction|GetAvatarasFortPlayerPawn) (Variables|Default|GetCooldownEndSoundCue))))))))

(event Custom|UpdatePingsClient
  (Ability|HasAuthority
    (:False
      (Class|WBPBuildingPoints|UpdatePips (Variables|Default|GetBuildingPointsWidget) (Variables|Default|GetBuildingPoints)))))
