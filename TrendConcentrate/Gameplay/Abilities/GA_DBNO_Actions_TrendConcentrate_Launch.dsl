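; Gameplay ability that plays an impact launch reaction on a downed (DBNO) player pawn.
; On activation it adds impulse tags, links an anim layer on non-authority roles and plays a montage chosen from the event instigator, with notifies setting chest-up or chest-down landing.
; Vehicle-tag and movement-mode wait tasks, plus the pawn death event, end the ability.
; Ending removes the tags, unlinks the layer and stops the stored movement tasks.
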(event Ability|EventActivateAbilityFromEvent
  (params (EventData :type Gameplay%20Event%20Data%20Structure%20%28by%20ref%29))
  (bind (_eventtag _instigator) (Utilities|Struct|BreakGameplayEventData EventData))
  (bind _returnvalue (Ability|GetAvatarActorfromActorInfo))
  (Ability|GameplayEffect|AddLooseGameplayTags _returnvalue (Variables|Default|GetImpulseTags) true)
  (bind _asfort_player_pawn_athena (Utilities|Casting|CastToFortPlayerPawnAthena _returnvalue)
    (:then
      (EventDispatchers|BindEventtoOnFortPlayerDied _asfort_player_pawn_athena (EventDispatchers|CreateEvent :!event Custom|On%20Fort%20Player%20Pawn%20Athena%20Died)))
    (:CastFailed
      _))
  (CallFunction|LinkAnimLayer)
  (bind (_impact_montage _frontimpact) (CallFunction|GetImpactMontage _instigator))
  (Variables|Default|SetFront%20Impact _frontimpact)
  (Utilities|IsValid _impact_montage
    (:Is%20Valid
      (bind _output_get (Variables|Default|SetImpact%20Montage _impact_montage))
      (bind (_asynctaskproxy _notifytag) (Ability|Tasks|PlayMontageAndWaitNotify "VehicleImpactMontageTask" _output_get)
        (:then
          ; Make sure we end this ability if we land on a vehicle regardless of movement mode changes
          (Ability|Tasks|WaitGameplayTagAdd "(TagName=\"Gameplay.Action.Player.AthenaInVehicle\")"
            (:then
              (bind _asynctaskproxy_2 (Ability|Tasks|WaitMovementModeChange "MOVE_Walking" :!label _label)
                (:then
                  (Variables|Default|SetMovement%20Mode%20Change%20Async%20Task%20Walking _asynctaskproxy_2)
                  (bind _asynctaskproxy_3 (Ability|Tasks|WaitMovementModeChange "MOVE_NavWalking")
                    (:then
                      (Variables|Default|SetMovement%20Mode%20Change%20Async%20Task%20Navmesh%20Walking _asynctaskproxy_3)
                      (bind _asynctaskproxy_4 (Ability|Tasks|WaitMovementModeChange "MOVE_Swimming")
                        (:then
                          (Variables|Default|SetMovement%20Mode%20Change%20Async%20Task%20Swimming _asynctaskproxy_4)
                          (bind _asynctaskproxy_5 (Ability|Tasks|WaitMovementModeChange "MOVE_Custom")
                            (:then
                              (Variables|Default|SetMovement%20Mode%20Change%20Async%20Task%20Custom _asynctaskproxy_5))
                            (:OnChange
                              _)))
                        (:OnChange
                          _)))
                    (:OnChange
                      _)))
                (:OnChange
                  _)))
            (:Added
              _))
          (Ability|EndAbility))
        (:OnNotifyBegin
          (if (== _notifytag "(TagName=\"Gameplay.Action.Player.TrendConcentrate.LandUp\")")
            (Variables|Default|SetChestFacingUp true)
            (elif (== _notifytag "(TagName=\"Gameplay.Action.Player.TrendConcentrate.LandDown\")")
              (Variables|Default|SetChestFacingUp false)))))
      (goto _label))
    (:Is%20Not%20Valid
      (Ability|EndAbility))))

(event Ability|EventOnEndAbility
  (params (bWasCancelled :type Boolean))
  (CallFunction|UnlinkAnimLayer)
  (Ability|GameplayEffect|RemoveLooseGameplayTags (Ability|GetAvatarActorfromActorInfo) (Variables|Default|GetImpulseTags) true)
  (CallFunction|EndTasks))

(event Ability|EventOnAbilityRemoved
  (Ability|EndAbility))

(event Custom|On%20Fort%20Player%20Pawn%20Athena%20Died
  (params (FortPlayerPawnAthena :type Fort%20Player%20Pawn%20Athena%20Object%20Reference))
  (if (== FortPlayerPawnAthena (Ability|GetAvatarActorfromActorInfo))
    (Ability|EndAbility)))

(fn EndTasks
  (GameplayTasks|EndTask (Variables|Default|GetMovement%20Mode%20Change%20Async%20Task%20Walking))
  (GameplayTasks|EndTask (Variables|Default|GetMovement%20Mode%20Change%20Async%20Task%20Navmesh%20Walking))
  (GameplayTasks|EndTask (Variables|Default|GetMovement%20Mode%20Change%20Async%20Task%20Swimming))
  (GameplayTasks|EndTask (Variables|Default|GetMovement%20Mode%20Change%20Async%20Task%20Custom)))

(fn CanActivateAbility
  (params (ActorInfo :type Gameplay%20Ability%20Actor%20Info%20Structure) (Handle :type Gameplay%20Ability%20Spec%20Handle%20Structure))
  (bind (_owneractor _avataractor) (Utilities|Struct|BreakGameplayAbilityActorInfo ActorInfo))
  (bind _asfort_pawn (Utilities|Casting|CastToFortPawn _avataractor))
  (Utilities|IsValid _asfort_pawn
    (:Is%20Valid
      (return 0 (Variables|/Script/FortniteGame.FortPawn|GetbIsDBNO _asfort_pawn)))
    (:Is%20Not%20Valid
      (return 0 false))))

(fn LinkAnimLayer
  (bind _asfort_pawn (Utilities|Casting|CastToFortPawn (Ability|GetAvatarActorfromActorInfo)))
  (if (Utilities|Enum|NotEqual(Enum) (Networking|GetLocalRole _asfort_pawn) "ROLE_Authority")
    (Animation|LinkAnimLayer _asfort_pawn (Variables|Default|GetAnim%20Layer))))

(fn SignedAngle
  (params (V_X :type Float%20%28double-precision%29) (V_Y :type Float%20%28double-precision%29) (W_X :type Float%20%28double-precision%29) (W_Y :type Float%20%28double-precision%29))
  (return (Math|Trig|Atan2(Radians) (- (* V_X W_Y) (* V_Y W_X)) (+ (* V_X W_X) (* V_Y W_Y)))))

(fn UnlinkAnimLayer
  (bind _asfort_pawn (Utilities|Casting|CastToFortPawn (Ability|GetAvatarActorfromActorInfo)))
  (if (Utilities|Enum|NotEqual(Enum) (Networking|GetLocalRole _asfort_pawn) "ROLE_Authority")
    (Animation|UnlinkAnimLayer _asfort_pawn (Variables|Default|GetAnim%20Layer))))
