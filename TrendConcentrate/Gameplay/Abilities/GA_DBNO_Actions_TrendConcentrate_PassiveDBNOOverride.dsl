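; Passive gameplay ability for the TrendConcentrate mode that gives the avatar player pawn a DBNO (down-but-not-out) override tag.
; On activation it caches the Fort player pawn and its ability system component, then self-applies a configurable gameplay effect and stores the returned handle.
; On ability end it removes that effect by handle, so the override lasts only while the ability is active.
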(event Ability|EventActivateAbility
  (Variables|Default|SetPlayerPawn (CallFunction|GetAvatarasFortPlayerPawn))
  (bind _playerpawn (Variables|Default|GetPlayerPawn)
    (:then
      (Variables|Default|SetAbilityComponent (Ability|GetAbilitySystemComponent _playerpawn))
      (bind _abilitycomponent (Variables|Default|GetAbilityComponent)
        (:then
          ; Give Player DBNO Override Tag
          (bind _returnvalue (GameplayEffects|ApplyGameplayEffectToSelf :self _abilitycomponent :GameplayEffectClass (Variables|Default|GetGameplay%20Effect%20Class) :Level 0.0))
          (Variables|Default|SetGEHandle _returnvalue))
        (:else)))
    (:else)))

(event Ability|EventOnEndAbility
  (params (bWasCancelled :type Boolean))
  (bind _abilitycomponent (Variables|Default|GetAbilityComponent)
    (:then
      (GameplayEffects|RemoveActiveGameplayEffect _abilitycomponent (Variables|Default|GetGEHandle) -1))
    (:else)))
