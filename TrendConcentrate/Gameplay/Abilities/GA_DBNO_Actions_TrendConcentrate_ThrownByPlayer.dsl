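; Gameplay ability for a downed (DBNO) player being thrown by another player, activated from an event and allowed only while the avatar pawn is DBNO.
; It plays the thrown montage and jumps to its Exit section when the character's movement mode changes.
; Montage completion or gaining the in-vehicle tag stops the montage and ends the ability, covering vehicle landings regardless of movement mode.
; Ending unbinds the movement delegate, and holster application is always refused.
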(event Ability|EventActivateAbilityFromEvent
  (params (EventData :type Gameplay%20Event%20Data%20Structure%20%28by%20ref%29))
  (Ability|Tasks|PlayMontageAndWaitNotify "ThrownMontageTask" (Variables|Default|GetThrown%20Montage) 1.0 "Default"
    (:then
      ; Make sure we end this ability if we land on a vehicle regardless of movement mode changes
      (Ability|Tasks|WaitGameplayTagAdd "(TagName=\"Gameplay.Action.Player.AthenaInVehicle\")"
        (:then
          (bind _outfortplayerpawn (CallFunction|GetAvatarasValidFortPlayerPawn)
            (:ValidFortPlayerPawn
              (Character|AssignMovementModeChangedDelegate _outfortplayerpawn (delegate Custom|MovementModeChangedDelegate_Event)))))
        (:Added
          _)))
    (:OnCompleted
      _))
  (Ability|Animation|MontageStop -1.0)
  (Ability|EndAbility))

(event Custom|MovementModeChangedDelegate_Event
  (params (Character :type Character%20Object%20Reference) (PrevMovementMode :type EMovementMode%20Enum :default "MOVE_None") (PreviousCustomMode :type Byte))
  (Ability|Animation|MontageJumptoSection "Exit"))

(event Ability|EventOnEndAbility
  (params (bWasCancelled :type Boolean))
  (bind _outfortplayerpawn (CallFunction|GetAvatarasValidFortPlayerPawn)
    (:ValidFortPlayerPawn
      (Character|UnbindEventfromMovementModeChangedDelegate _outfortplayerpawn (EventDispatchers|CreateEvent :!event Custom|MovementModeChangedDelegate_Event)))))

(fn CanApplyHolster
  (params (HolsterId :type Name%20%28by%20ref%29 :default "None"))
  (return false))

(fn CanActivateAbility
  (params (ActorInfo :type Gameplay%20Ability%20Actor%20Info%20Structure) (Handle :type Gameplay%20Ability%20Spec%20Handle%20Structure))
  (bind (_owneractor _avataractor) (Utilities|Struct|BreakGameplayAbilityActorInfo ActorInfo))
  (bind _asfort_pawn (Utilities|Casting|CastToFortPawn _avataractor))
  (Utilities|IsValid _asfort_pawn
    (:Is%20Valid
      (return 0 (Variables|/Script/FortniteGame.FortPawn|GetbIsDBNO _asfort_pawn)))
    (:Is%20Not%20Valid
      (return 0 false))))
