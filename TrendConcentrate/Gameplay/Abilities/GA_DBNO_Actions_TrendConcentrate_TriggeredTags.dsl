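; Gameplay ability that adjusts the player pawn's state while active, apparently during down-but-not-out (DBNO) actions, judging by its name.
; On activation it removes a configured set of loose tags and applies a gameplay effect to the pawn's ability system, noted as blocking the driver seat.
; On end, cancelled or not, it adds those tags back and removes the effect through its stored handle.
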(event Ability|EventActivateAbility
  (Variables|Default|SetPlayerPawn (CallFunction|GetAvatarasFortPlayerPawn))
  (bind _playerpawn (Variables|Default|GetPlayerPawn)
    (:then
      (Ability|GameplayEffect|RemoveLooseGameplayTags _playerpawn (Variables|Default|GetTagsToRemove) true)
      ;[1 1 1] Block driver seat
      (Variables|Default|SetAbilityComponent (Ability|GetAbilitySystemComponent (Variables|Default|GetPlayerPawn)))
      (bind _abilitycomponent (Variables|Default|GetAbilityComponent)
        (:then
          (bind _returnvalue (GameplayEffects|ApplyGameplayEffectToSelf :self _abilitycomponent :GameplayEffectClass (Variables|Default|GetGameplay%20Effect%20Class) :Level 0.0))
          (Variables|Default|SetGEHandle _returnvalue))
        (:else)))
    (:else)))

(event Ability|EventOnEndAbility
  (params (bWasCancelled :type Boolean))
  (bind _playerpawn (Variables|Default|GetPlayerPawn)
    (:then
      (Ability|GameplayEffect|AddLooseGameplayTags _playerpawn (Variables|Default|GetTagsToRemove) true)
      (bind _abilitycomponent (Variables|Default|GetAbilityComponent)
        (:then
          (GameplayEffects|RemoveActiveGameplayEffect _abilitycomponent (Variables|Default|GetGEHandle) -1))
        (:else)))
    (:else)))
