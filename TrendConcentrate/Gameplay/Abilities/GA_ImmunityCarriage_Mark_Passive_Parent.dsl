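; Parent passive ability that marks the nearest rez-seat-enabled Immunity Carriage for the player and ends when the avatar's DBNO tag is removed.
; On activation it filters the game state's reboot spawn machines for such vehicles and picks the closest as LockActor.
; After a net sync, cosmetic-capable instances register a marker on it if within MaxPingDistance.
; Separate trail events compute direction on the server and replicate a keycard-style Niagara pointer, though nothing here calls them.
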(event Ability|EventActivateAbility
  (Ability|Tasks|WaitGameplayTagRemoveFromActor (Ability|GetAvatarActorfromActorInfo) "(TagName=\"Gameplay.Status.DBNO\")"
    (:then
      (CallFunction|InitAbility)
      (Utilities|Array|Clear (Variables|Default|GetSpawnMachinesInsideSafeZone))
      (for _array_element (Resurrection|GetSpawnMachineActors (Actor|GetComponentbyClass (Game|GetGameState) "/Script/FortniteGame.FortGameStateComponent_Reboot") true)
        (bind _asvalet_immunity_carriage_vehicle (Utilities|Casting|CastToValet_ImmunityCarriage_Vehicle _array_element))
        (if (Class|ValetImmunityCarriageVehicle|GetIsRezSeatEnabled _asvalet_immunity_carriage_vehicle)
          (Utilities|Array|AddUnique (Variables|Default|GetSpawnMachinesInsideSafeZone) _asvalet_immunity_carriage_vehicle)))
      (Variables|Default|SetLockActor (Fortnite|GetClosestActorfromArray (Ability|GetAvatarActorfromActorInfo) (Variables|Default|GetSpawnMachinesInsideSafeZone)))
      (Ability|Tasks|WaitNetSync "BothWait"
        (:OnSync
          (Networking|CanExecuteCosmeticEvents
            (:True
              (bind _lockactor (Variables|Default|GetLockActor)
                (:then
                  (if (<= (Math|Vector|VectorLength (- (Transformation|GetActorLocation (Variables|Default|GetLockActor)) (Transformation|GetActorLocation (Ability|GetAvatarActorfromActorInfo)))) (Variables|Default|GetMaxPingDistance))
                    (bind (_playerbankid _bankinstanceid) (Fortnite|RegisterSoundIndicatorPlacementContext (Variables|Default|GetPlayerController) _lockactor "0, 0, 0" false))
                    (Variables|SetMarker_PlayerID (Utilities|Struct|MakeMarkerID _playerbankid _bankinstanceid))))
                (:else)))))))
    (:Removed
      (Ability|EndAbility))))

(event Custom|DrawTrailToVault
  ;[1 1 1] From the server do the math and pass the rotation to the client. This is done so that if the actor is not relevant to the client the server still points in the correct direction.
  (if (Networking|IsServer)
    (CallFunction|ReplicatedDrawTrailtoVault (Math|Rotator|MakeRotfromX (* (- (Transformation|GetActorLocation (CallFunction|GetAvatarasFortPlayerPawn)) (Transformation|GetActorLocation (Variables|Default|GetLockActor))) -1.0)) (Transformation|GetActorLocation (Variables|Default|GetLockActor)))))

(event Custom|ReplicatedDrawTrailToVault
  (params (Rotation :type Rotator) (LockActorPos :type Vector))
  ;[1 1 1] From the server do the math and pass the rotation to the client. This is done so that if the actor is not relevant to the client the server still points in the correct direction.
  (bind _returnvalue (Niagara|SpawnSystemAttached "/Game/Athena/Items/EnvironmentalItems/Locks/Keycard/FX/NS_Keycard_LockDirection.NS_Keycard_LockDirection" (Variables|/Script/Engine.Character|GetMesh (CallFunction|GetAvatarasFortPlayerPawn)) "root" "0.000000,0.000000,125.000000" "0, 0, 0" "SnapToTarget" true))
  (Effects|Components|ParticleSystem|SetVectorParameter _returnvalue "LockActorPos" LockActorPos))

(fn InitAbility
  (bind _asfort_player_controller_athena (Utilities|Casting|CastToFortPlayerControllerAthena (Fortnite|GetFortPlayerControllerfromActor (CallFunction|GetAvatarasFortPlayerPawn))))
  (Variables|Default|SetPlayerController _asfort_player_controller_athena))
